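# Physics asset for the male medium Nautilus player body, simulating its dangling rope, air hose and seaweed attachments.
# Capsule and box bodies on pelvis, upper spine, right clavicle and forearm, and left calf serve as anchors, linked by free-swinging constraints.
# Tightly limited chains hang from them: four rope segments and an air hose on the spine, seaweed on the forearm and calf.
# The rope root locks twist and one swing axis.
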PhysicsAsset /BRCosmetics/Characters/Player/Male/Medium/Bodies/M_MED_Nautilus/Meshes/M_MED_Nautilus_Physics
bodies {
 calf_l [{shapeType=Capsule, length=26.4074, radius=8.215, center=<-28.1233 -1.8037 -0.4602>, rotation=<rot 90 0 0>}]
 clavicle_r [{shapeType=Capsule, length=9.937, radius=8.7953, center=<-10.9316 -0.1467 2.4838>, rotation=<rot 0 -90.7094 -89.9999>}]
 dyn_air_hose_r [{shapeType=Box, extentX=3.0241, extentY=13.2456, extentZ=5.0028, center=<-0.0733 -0.3737 -0.62>}]
 dyn_arm_seaweed01 [{shapeType=Capsule, length=4.3623, radius=2.6826, center=<3.6404 -0.1256 -0.1193>, rotation=<rot 90 0 0>}]
 dyn_arm_seaweed02 [{shapeType=Capsule, length=4.0104, radius=2.1263, center=<4.5614 0.0573 -0.0647>, rotation=<rot 90 0 0>}]
 dyn_arm_seaweed03 [{shapeType=Capsule, length=5.2452, radius=1.4965, center=<6.0415 0.0468 0.8928>, rotation=<rot 90 0 0>}]
 dyn_leg_seaweed01 [{shapeType=Capsule, length=2.7435, radius=1.9644, center=<3.2469 0.083 -0.4074>, rotation=<rot 0.0001 -90.6065 -89.9998>}]
 dyn_leg_seaweed02 [{shapeType=Capsule, length=7.2319, radius=1.5902, center=<2.5548 -0.6231 0.3507>, rotation=<rot 90 0 0>}]
 dyn_rope01 [{shapeType=Box, extentX=4.2548, extentY=20.7391, extentZ=34.2705, center=<10.7332 -0.7902 -1.278>, rotation=<rot 1.0054 -89.9999 0.0203>}]
 dyn_rope02 [{shapeType=Box, extentX=5.01, extentY=13.01, extentZ=34.01, center=<9.9617 0.0002 0.0003>, rotation=<rot 0 -90 0>}]
 dyn_rope03 [{shapeType=Box, extentX=4.01, extentY=11.51, extentZ=34.01, center=<9.1218 0.0002 -0.0001>, rotation=<rot 0 -90 0>}]
 dyn_rope04 [{shapeType=Box, extentX=3.01, extentY=13.01, extentZ=34.01, center=<6.4145 0.0002 0>, rotation=<rot 0 -90 0>}]
 lowerarm_r [{shapeType=Box, extentX=13.2398, extentY=10.3427, extentZ=10.9879, center=<-16.6785 -0.1673 -2.9541>}]
 pelvis [{shapeType=Capsule, length=14.9126, radius=16.8374, center=<5.2632 -0.0001 0.0719>, rotation=<rot 89.517 179.9998 179.9998>}]
 spine_05 [{shapeType=Box, extentX=29.3321, extentY=11.453, extentZ=39.445, center=<2.7342 0.4038 1.6754>, rotation=<rot -4.6809 -164.6242 -88.2005>}]
}
constraints [
 {
  bone1Name=spine_05
  bone2Name=pelvis
  swing1LimitDegrees=45
  swing1Motion=Free
  swing2LimitDegrees=45
  swing2Motion=Free
  twistLimitDegrees=45
  twistMotion=Free
 }
 {
  bone1Name=dyn_rope01
  bone2Name=spine_05
  swing1LimitDegrees=7
  swing2LimitDegrees=1
  swing2Motion=Locked
  twistLimitDegrees=1
  twistMotion=Locked
 }
 {bone1Name=dyn_rope02, bone2Name=dyn_rope01, swing1LimitDegrees=20, swing2LimitDegrees=1, twistLimitDegrees=1}
 {bone1Name=dyn_rope03, bone2Name=dyn_rope02, swing1LimitDegrees=20, swing2LimitDegrees=1, twistLimitDegrees=1}
 {bone1Name=dyn_rope04, bone2Name=dyn_rope03, swing1LimitDegrees=20, swing2LimitDegrees=1, twistLimitDegrees=1}
 {bone1Name=dyn_leg_seaweed01, bone2Name=calf_l, swing1LimitDegrees=3, swing2LimitDegrees=15, twistLimitDegrees=1}
 {bone1Name=dyn_leg_seaweed02, bone2Name=dyn_leg_seaweed01, swing1LimitDegrees=3, swing2LimitDegrees=15, twistLimitDegrees=1}
 {
  bone1Name=calf_l
  bone2Name=pelvis
  swing1LimitDegrees=45
  swing1Motion=Free
  swing2LimitDegrees=45
  swing2Motion=Free
  twistLimitDegrees=45
  twistMotion=Free
 }
 {bone1Name=dyn_arm_seaweed01, bone2Name=lowerarm_r, swing1LimitDegrees=10, swing2LimitDegrees=20, twistLimitDegrees=2}
 {bone1Name=dyn_arm_seaweed02, bone2Name=dyn_arm_seaweed01, swing1LimitDegrees=15, swing2LimitDegrees=30, twistLimitDegrees=2}
 {bone1Name=dyn_arm_seaweed03, bone2Name=dyn_arm_seaweed02, swing1LimitDegrees=15, swing2LimitDegrees=30, twistLimitDegrees=2}
 {
  bone1Name=lowerarm_r
  bone2Name=clavicle_r
  swing1LimitDegrees=45
  swing1Motion=Free
  swing2LimitDegrees=45
  swing2Motion=Free
  twistLimitDegrees=45
  twistMotion=Free
 }
 {bone1Name=dyn_air_hose_r, bone2Name=spine_05, swing1LimitDegrees=3, swing2LimitDegrees=15, twistLimitDegrees=1}
 {
  bone1Name=clavicle_r
  bone2Name=spine_05
  swing1LimitDegrees=45
  swing1Motion=Free
  swing2LimitDegrees=45
  swing2Motion=Free
  twistLimitDegrees=45
  twistMotion=Free
 }
]
