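# Physics asset for the belt mesh of the male medium Nautilus body cosmetic, simulating its dangling weight-belt rings and seaweed.
# A pelvis capsule anchors the rotation-locked belt master box and limited-swing seaweed capsules, including a two-link strand chain.
# The front ring has tight angular limits on the belt master.
# The back ring is rotation-locked to a helper body that rotates freely on the master.
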PhysicsAsset /BRCosmetics/Characters/Player/Male/Medium/Bodies/M_MED_Nautilus/Meshes/M_MED_Nautilus_belt_Physics
bodies {
 dyn_belt_ring_front_bottom [{shapeType=Box, extentX=1.396, extentY=6.9872, extentZ=7.507, center=<2.7268 -0.179 0>, rotation=<rot 0 -90 0.0008>}]
 dyn_butt_seaweed [{shapeType=Capsule, length=8.5, radius=2.75, center=<5.0127 -0.1282 0>}]
 dyn_butt_seaweed_strand01 [{shapeType=Capsule, length=6.755, radius=1.505, center=<3.5253 0.0002 0.0679>, rotation=<rot -88.8959 0 0>}]
 dyn_butt_seaweed_strand02 [{shapeType=Capsule, length=5.505, radius=1.755, center=<5.5745 0 0.2759>, rotation=<rot -87.1659 0 0>}]
 dyn_weight_belt_back_ring [{shapeType=Box, extentX=1.7911, extentY=6.4689, extentZ=6.5081, center=<-0.0939 -1.6144 0.1074>, rotation=<rot -0.0829 -89.9982 -0.9404>}]
 dyn_weight_belt_back_ring_const [{shapeType=Box, extentX=1.01, extentY=5.8715, extentZ=5.2948, rotation=<rot 0 -90 0>}]
 dyn_weight_belt_master [{shapeType=Box, extentX=13.7164, extentY=33.9147, extentZ=37.389, center=<-3.2239 -0.2597 -0.7377>}]
 pelvis [{shapeType=Capsule, length=17, radius=14.75, center=<5.2632 -0.0001 0.0719>, rotation=<rot 89.517 179.9998 179.9998>}]
}
constraints [
 {
  bone1Name=dyn_weight_belt_master
  bone2Name=pelvis
  swing1LimitDegrees=45
  swing1Motion=Locked
  swing2LimitDegrees=45
  swing2Motion=Locked
  twistLimitDegrees=45
  twistMotion=Locked
 }
 {bone1Name=dyn_butt_seaweed, bone2Name=pelvis, swing1LimitDegrees=20, swing2LimitDegrees=1, twistLimitDegrees=1}
 {bone1Name=dyn_butt_seaweed_strand02, bone2Name=dyn_butt_seaweed_strand01, swing1LimitDegrees=15, swing2LimitDegrees=3, twistLimitDegrees=3}
 {bone1Name=pelvis, bone2Name=dyn_butt_seaweed_strand01, swing1LimitDegrees=15, swing2LimitDegrees=5, twistLimitDegrees=1}
 {bone1Name=dyn_belt_ring_front_bottom, bone2Name=dyn_weight_belt_master, swing1LimitDegrees=5, swing2LimitDegrees=3, twistLimitDegrees=1}
 {
  bone1Name=dyn_weight_belt_back_ring_const
  bone2Name=dyn_weight_belt_master
  swing1LimitDegrees=45
  swing1Motion=Free
  swing2LimitDegrees=45
  swing2Motion=Free
  twistLimitDegrees=45
  twistMotion=Free
 }
 {
  bone1Name=dyn_weight_belt_back_ring
  bone2Name=dyn_weight_belt_back_ring_const
  swing1LimitDegrees=45
  swing1Motion=Locked
  swing2LimitDegrees=45
  swing2Motion=Locked
  twistLimitDegrees=45
  twistMotion=Locked
 }
]
