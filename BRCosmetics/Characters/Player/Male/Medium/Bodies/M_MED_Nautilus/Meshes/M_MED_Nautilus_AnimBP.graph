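# Post-copy secondary-motion AnimBP for the M_MED_Nautilus body, animating its weight belt and rope bones.
# Copies the attached parent's pose, simulates RigidBody physics in root space, then constrains the back belt ring across spine, clavicles and belt master and aims a front ring.
# Bool-blended constraints pin dyn_rope03 to the pelvis during tactical sprint and skydive dive.
# A second belt RigidBody and three back-belt LookAts are unconnected to the output; the update event is empty.
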AnimBlueprint /BRCosmetics/Characters/Player/Male/Medium/Bodies/M_MED_Nautilus/Meshes/M_MED_Nautilus_AnimBP
target_skeleton=&/M_MED_Nautilus_Skeleton
graphs [
 AnimGraph {nodes [
   Look At - Bone: dyn_weight_belt_back_ring_const(Alpha=1) {
    alphaBoolBlend {blendOption=Linear}
    boneToModify {boneName=dyn_weight_belt_back_ring_const}
    lookAtLocation=<0 -2>
    lookAtTarget {boneReference {boneName=dyn_weight_belt_back_bottom}}
    lookAt_Axis {axis=<-1 0>}
   }
   n1 = Look At - Bone: dyn_weight_belt_back_bottom(Alpha=1) {
    lookAtClamp=15
    alphaBoolBlend {blendOption=Linear}
    boneToModify {boneName=dyn_weight_belt_back_bottom}
    lookAtLocation=<0 -5>
    lookAtTarget {boneReference {boneName=dyn_weight_belt_back_top}}
    lookAt_Axis {axis=<1 0>}
   }
   Look At - Bone: dyn_weight_belt_back_top_aim(ComponentPose=@n1, Alpha=1) {
    alphaBoolBlend {blendOption=Linear}
    boneToModify {boneName=dyn_weight_belt_back_top_aim}
    lookAtLocation=<0 -1>
    lookAtTarget {boneReference {boneName=dyn_weight_belt_back_top}}
    lookAt_Axis {axis=<1 0>}
   }
   RigidBody(Alpha=0.5) {
    bOverrideWorldGravity=true
    overridePhysicsAsset=&/M_MED_Nautilus_belt_Physics
    simulationSpace=BaseBoneSpace
    alphaBoolBlend {blendOption=Linear}
    baseBoneRef {boneName=root}
    overrideWorldGravity {z=-980}
   }
   n2 = RigidBody(ComponentPose=Local To Component(LocalPose=Copy Pose From Mesh() {bUseAttachedParent=true})) {
    bOverrideWorldGravity=true
    overridePhysicsAsset=&/M_MED_Nautilus_Physics
    simulationSpace=BaseBoneSpace
    alphaBoolBlend {blendOption=Linear}
    baseBoneRef {boneName=root}
    overrideWorldGravity {z=-980}
   }
   n3 = "Constraint - dyn_weight_belt_back_ring_const (PPPPPPPP)"!(
    ComponentPose=@n2
    Alpha=1
    ConstraintWeights_0=0.0353
    ConstraintWeights_1:float=0.0833
    ConstraintWeights_2:float=0.0777
    ConstraintWeights_3:float=0.2837
    ConstraintWeights_4:float=0.0549
    ConstraintWeights_5:float=0.2146
    ConstraintWeights_6:float=0.0388
    ConstraintWeights_7:float=0.2117
   ) -> (Pose) {
    alphaBoolBlend {blendOption=Linear}
    boneToModify {boneName=dyn_weight_belt_back_ring_const}
    constraintSetup [
     {targetBone {boneName=clavicle_l}}
     {targetBone {boneName=spine_01}}
     {targetBone {boneName=spine_02}}
     {targetBone {boneName=spine_03}}
     {targetBone {boneName=spine_04}}
     {targetBone {boneName=spine_05}}
     {targetBone {boneName=clavicle_r}}
     {targetBone {boneName=dyn_weight_belt_master}}
    ]
    constraintWeights [1,1,1,1,1,1,1,1]
   }
   n4 = Look At - Bone: dyn_belt_ring_front_top(ComponentPose=@n3, Alpha=1) {
    alphaBoolBlend {blendOption=Linear}
    boneToModify {boneName=dyn_belt_ring_front_top}
    lookAtLocation=<0 3>
    lookAtTarget {boneReference {boneName=dyn_weight_belt_front}}
    lookAt_Axis {axis=<1 0>}
   }
   n5 = "Constraint - dyn_rope03 (P)"!(ComponentPose=@n4, bAlphaBoolEnabled:bool=Get bIsTacticalSprinting(), ConstraintWeights_0=1) -> (Pose) {
    alphaInputType=Bool
    alphaBoolBlend {blendInTime=0.5, blendOption=Linear, blendOutTime=1}
    boneToModify {boneName=dyn_rope03}
    constraintSetup [{targetBone {boneName=pelvis}}]
    constraintWeights [1]
   }
   n6 = "Constraint - dyn_rope03 (P)"(ComponentPose=@n5, bAlphaBoolEnabled=Get bIsSkydiveDiveMode(), ConstraintWeights_0=1) {
    alphaInputType=Bool
    alphaBoolBlend {blendInTime=1, blendOption=Linear, blendOutTime=2}
    boneToModify {boneName=dyn_rope03}
    constraintSetup [{targetBone {boneName=pelvis}}]
    constraintWeights [1]
   }
   Output Pose(Result=Component To Local(ComponentPose=@n6)) {layerGroup=None}
  ]}
 EventGraph {nodes [Event BlueprintUpdateAnimation()]}
]
variables []
